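# Post-process Anim Blueprint for the MarlinYeast biped cosmetic companion, adding secondary motion to its cape and feather.
# Weighted constraints and additive offsets position cape roots and rapier attach, and derive ring fingers from middle and pinky.
# That cached pose feeds chained AnimDynamics for cape, attach_emote_01 and feather_01, then the output.
# Some dynamics and pose-asset nodes are unconnected; event-computed cape noise and leg alpha go unread here.
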AnimBlueprint /CosmeticCompanions/Assets/Biped/MarlinYeast/Meshes/MarlinYeast_Biped_Post_AnimBP
target_skeleton=&/MarlinYeast_Biped_Skeleton
graphs [
 AnimGraph {nodes [
   n1 = Anim Dynamics - Bone: feather_01!(ComponentPose, Alpha=1, GravityOverride:Vector="0,0,-300") -> (Pose) {
    angularDampingOverride=0.9999
    angularSpringConstant=3000
    bAngularSpring=true
    bOverrideAngularDamping=true
    bUseGravityOverride=true
    gravityScale=0.35
    simulationSpace=RootRelative
    alphaScaleBiasClamp {interpSpeedDecreasing=7, interpSpeedIncreasing=7}
    boundBone {boneName=feather_01}
    componentAppliedLinearAccClamp=<1000 1000 1000>
    componentLinearAccScale=<1 1 1>
    componentLinearVelScale=<1 1 1>
    externalForce {z=-980}
    gravityOverride {z=-980}
    physicsBodyDefinitions [{
      boundBone {boneName=feather_01}
      boxExtents=<15 25 25>
      constraintSetup {angularLimitsMax=<0 5 15>, angularLimitsMin=<0 -5 -15>, angularTarget=<1 0 0>}
      localJointOffset=<4 0 0>
     }]
    simSpaceSettings {maxAngularAcceleration=35, maxAngularVelocity=7, simSpaceAngularAlpha=1}
   }
   Anim Dynamics - Bone: feather_02!(ComponentPose=@n1, Alpha=1, GravityOverride:Vector="0,0,-300") -> (Pose) {
    angularDampingOverride=0.9999
    angularSpringConstant=3000
    bAngularSpring=true
    bOverrideAngularDamping=true
    bUseGravityOverride=true
    gravityScale=0.35
    simulationSpace=RootRelative
    alphaScaleBiasClamp {interpSpeedDecreasing=7, interpSpeedIncreasing=7}
    boundBone {boneName=feather_02}
    componentAppliedLinearAccClamp=<1000 1000 1000>
    componentLinearAccScale=<1 1 1>
    componentLinearVelScale=<1 1 1>
    externalForce {z=-980}
    gravityOverride {z=-980}
    physicsBodyDefinitions [{
      boundBone {boneName=feather_02}
      boxExtents=<15 25 25>
      constraintSetup {angularLimitsMax=<5 15 30>, angularLimitsMin=<-5 -15 -30>, angularTarget=<1 0 0>}
      localJointOffset=<4 0 0>
     }]
    simSpaceSettings {maxAngularAcceleration=35, maxAngularVelocity=7, simSpaceAngularAlpha=1}
   }
   Anim Dynamics - Bone: rapier_attach!(ComponentPose, Alpha=1, GravityOverride:Vector="0,0,-980") -> (Pose) {
    angularDampingOverride=0.9999
    angularSpringConstant=3000
    bAngularSpring=true
    bOverrideAngularDamping=true
    bUseGravityOverride=true
    gravityScale=0.35
    simulationSpace=RootRelative
    boundBone {boneName=rapier_attach}
    componentAppliedLinearAccClamp=<1000 1000 1000>
    componentLinearAccScale=<1 1 1>
    componentLinearVelScale=<1 1 1>
    externalForce {z=-980}
    gravityOverride {z=-980}
    physicsBodyDefinitions [{
      boundBone {boneName=rapier_attach}
      boxExtents=<15 20 20>
      constraintSetup {angularLimitsMax=<0 10 10>, angularLimitsMin=<0 -5 -10>, angularTarget=<1 0 0>}
      localJointOffset=<0.25 0 0>
     }]
    simSpaceSettings {maxAngularAcceleration=30, maxAngularVelocity=6, simSpaceAngularAlpha=1}
   }
   "Constraint - feather_01 (P)"(Alpha=1, ConstraintWeights_0=1) {boneToModify {boneName=feather_01}, constraintSetup [{targetBone {boneName=attach_emote_01}}], constraintWeights [1]}
   Pose!(PoseAsset:Pose Asset Object Reference=Get PoseAsset() -> (PoseAsset:Pose Asset Object Reference), SourcePose:Pose Link Structure) -> (Pose:Pose Link Structure) {is=title AnimGraphNode_PoseBlendNode, bIgnorePoseAssetSourceAnimationCurves=true, blendOption=Linear}
   Output Pose(Result=Use cached pose 'DynamicsCache'())
   Save cached pose 'CRCache_Input'(Pose=Input Pose()) {cacheName=CRCache_Input}
   n2 = Anim Dynamics - Bone: cape_01_L!(ComponentPose=Local To Component(LocalPose=Use cached pose 'preDynamicsCache'()), Alpha=1, GravityOverride:Vector="0,0,-300") -> (Pose) {
    angularDampingOverride=1
    angularSpringConstant=1000
    bAngularSpring=true
    bOverrideAngularDamping=true
    bUseGravityOverride=true
    gravityScale=0.35
    simulationSpace=RootRelative
    boundBone {boneName=cape_01_L}
    componentAppliedLinearAccClamp=<1000 1000 1000>
    componentLinearAccScale=<1 3 1>
    componentLinearVelScale=<1 3 1>
    gravityOverride {z=-980}
    physicsBodyDefinitions [{
      boundBone {boneName=cape_01_L}
      boxExtents=<10 17.5 17.5>
      constraintSetup {angularLimitsMax=<5 5 15>, angularLimitsMin=<-5 -40 -15>, angularTarget=<1 0 0>}
      localJointOffset=<4.5 0 0>
     }]
    simSpaceSettings {maxAngularAcceleration=36, maxAngularVelocity=6, simSpaceAngularAlpha=1}
   }
   n3 = Anim Dynamics - Bone: cape_02_L!(ComponentPose=@n2, Alpha=1, GravityOverride:Vector="0,0,-100") -> (Pose) {
    angularDampingOverride=1
    angularSpringConstant=1000
    bAngularSpring=true
    bOverrideAngularDamping=true
    bUseGravityOverride=true
    gravityScale=0.35
    simulationSpace=RootRelative
    boundBone {boneName=cape_02_L}
    componentAppliedLinearAccClamp=<1000 1000 1000>
    componentLinearAccScale=<1 3 1>
    componentLinearVelScale=<1 3 1>
    gravityOverride {z=-980}
    physicsBodyDefinitions [{
      boundBone {boneName=cape_02_L}
      boxExtents=<7 12 12>
      constraintSetup {angularLimitsMax=<7.5 10 25>, angularLimitsMin=<-7.5 -60 -25>, angularTarget=<1 0 0>}
      localJointOffset=<5 0 0>
     }]
    simSpaceSettings {maxAngularAcceleration=36, maxAngularVelocity=6, simSpaceAngularAlpha=1}
   }
   n4 = Anim Dynamics - Bone: cape_01_C!(ComponentPose=@n3, Alpha=1, GravityOverride:Vector="0,0,-300") -> (Pose) {
    angularDampingOverride=1
    angularSpringConstant=1000
    bAngularSpring=true
    bOverrideAngularDamping=true
    bUseGravityOverride=true
    gravityScale=0.35
    simulationSpace=RootRelative
    boundBone {boneName=cape_01_C}
    componentAppliedLinearAccClamp=<1000 1000 1000>
    componentLinearAccScale=<1 3 1>
    componentLinearVelScale=<1 3 1>
    gravityOverride {z=-980}
    physicsBodyDefinitions [{
      boundBone {boneName=cape_01_C}
      boxExtents=<10 17.5 17.5>
      constraintSetup {angularLimitsMax=<5 2 15>, angularLimitsMin=<-5 -40 -15>, angularTarget=<1 0 0>}
      localJointOffset=<3.5 0 0>
     }]
    simSpaceSettings {maxAngularAcceleration=36, maxAngularVelocity=6, simSpaceAngularAlpha=1}
   }
   n5 = Anim Dynamics - Bone: cape_02_C!(ComponentPose=@n4, Alpha=1, GravityOverride:Vector="0,0,-100") -> (Pose) {
    angularDampingOverride=1
    angularSpringConstant=1000
    bAngularSpring=true
    bOverrideAngularDamping=true
    bUseGravityOverride=true
    gravityScale=0.35
    simulationSpace=RootRelative
    boundBone {boneName=cape_02_C}
    componentAppliedLinearAccClamp=<1000 1000 1000>
    componentLinearAccScale=<1 3 1>
    componentLinearVelScale=<1 3 1>
    gravityOverride {z=-980}
    physicsBodyDefinitions [{
      boundBone {boneName=cape_02_C}
      boxExtents=<7 12 12>
      constraintSetup {angularLimitsMax=<7.5 5 25>, angularLimitsMin=<-7.5 -60 -25>, angularTarget=<1 0 0>}
      localJointOffset=<4 0 0>
     }]
    simSpaceSettings {maxAngularAcceleration=36, maxAngularVelocity=6, simSpaceAngularAlpha=1}
    sphericalLimits [
     {limitRadius=5, drivingBone {boneName=tail_02}, sphereLocalOffset=<-5 0 0>}
     {limitRadius=8.5, drivingBone {boneName=pelvis}, sphereLocalOffset=<2 0 0>}
    ]
   }
   n6 = Anim Dynamics - Bone: cape_01_R!(ComponentPose=@n5, Alpha=1, GravityOverride:Vector="0,0,-300") -> (Pose) {
    angularDampingOverride=1
    angularSpringConstant=1000
    bAngularSpring=true
    bOverrideAngularDamping=true
    bUseGravityOverride=true
    gravityScale=0.35
    simulationSpace=RootRelative
    boundBone {boneName=cape_01_R}
    componentAppliedLinearAccClamp=<1000 1000 1000>
    componentLinearAccScale=<1 3 1>
    componentLinearVelScale=<1 3 1>
    gravityOverride {z=-980}
    physicsBodyDefinitions [{
      boundBone {boneName=cape_01_R}
      boxExtents=<10 17.5 17.5>
      constraintSetup {angularLimitsMax=<5 5 15>, angularLimitsMin=<-5 -40 -15>, angularTarget=<1 0 0>}
      localJointOffset=<4.5 0 0>
     }]
    simSpaceSettings {maxAngularAcceleration=36, maxAngularVelocity=6, simSpaceAngularAlpha=1}
   }
   n7 = Anim Dynamics - Bone: cape_02_R!(ComponentPose=@n6, Alpha=1, GravityOverride:Vector="0,0,-100") -> (Pose) {
    angularDampingOverride=1
    angularSpringConstant=1000
    bAngularSpring=true
    bOverrideAngularDamping=true
    bUseGravityOverride=true
    gravityScale=0.35
    simulationSpace=RootRelative
    boundBone {boneName=cape_02_R}
    componentAppliedLinearAccClamp=<1000 1000 1000>
    componentLinearAccScale=<1 3 1>
    componentLinearVelScale=<1 3 1>
    gravityOverride {z=-980}
    physicsBodyDefinitions [{
      boundBone {boneName=cape_02_R}
      boxExtents=<7 12 12>
      constraintSetup {angularLimitsMax=<7.5 10 25>, angularLimitsMin=<-7.5 -60 -25>, angularTarget=<1 0 0>}
      localJointOffset=<5 0 0>
     }]
    simSpaceSettings {maxAngularAcceleration=36, maxAngularVelocity=6, simSpaceAngularAlpha=1}
   }
   n8 = Anim Dynamics - Bone: attach_emote_01!(ComponentPose=@n7, Alpha=1, GravityOverride:Vector="0,0,-980") -> (Pose) {
    angularDampingOverride=0.9999
    angularSpringConstant=3000
    bAngularSpring=true
    bOverrideAngularDamping=true
    bUseGravityOverride=true
    bUseSphericalLimits=true
    gravityScale=0.35
    simulationSpace=RootRelative
    alphaScaleBiasClamp {bias=1, scale=-1}
    boundBone {boneName=attach_emote_01}
    componentAppliedLinearAccClamp=<1000 1000 1000>
    componentLinearAccScale=<1 1 1>
    componentLinearVelScale=<1 1 1>
    externalForce {z=-980}
    gravityOverride {z=-980}
    physicsBodyDefinitions [{
      boundBone {boneName=attach_emote_01}
      boxExtents=<15 20 20>
      constraintSetup {angularLimitsMax=<0 5 5>, angularLimitsMin=<0 -5 -5>, angularTarget=<1 0 0>}
      localJointOffset=<2 0 0>
     }]
    relativeSpaceBone {boneName=spine_03}
    simSpaceSettings {maxAngularAcceleration=30, maxAngularVelocity=6, simSpaceAngularAlpha=1}
   }
   n9 = Anim Dynamics - Bone: feather_01(ComponentPose=@n8, Alpha=1, GravityOverride="0,0,-300") {
    angularDampingOverride=0.9999
    angularSpringConstant=3000
    bAngularSpring=true
    bOverrideAngularDamping=true
    bUseGravityOverride=true
    gravityScale=0.35
    simulationSpace=RootRelative
    alphaScaleBiasClamp {interpSpeedDecreasing=7, interpSpeedIncreasing=7}
    boundBone {boneName=feather_01}
    componentAppliedLinearAccClamp=<1000 1000 1000>
    componentLinearAccScale=<1 1 1>
    componentLinearVelScale=<1 1 1>
    externalForce {z=150}
    gravityOverride {z=-980}
    physicsBodyDefinitions [{
      boundBone {boneName=feather_01}
      boxExtents=<12 20 20>
      constraintSetup {angularLimitsMax=<5 15 45>, angularLimitsMin=<-5 -15 -45>, angularTarget=<1 0 0>}
      localJointOffset=<6 0 0>
     }]
    simSpaceSettings {maxAngularAcceleration=35, maxAngularVelocity=7, simSpaceAngularAlpha=1}
   }
   Save cached pose 'DynamicsCache'(Pose=Component To Local(ComponentPose=@n9)) {cacheName=DynamicsCache}
   n10 = "Constraint - cape_01_L (PPPPP)"!(
    ComponentPose=Local To Component(LocalPose=Use cached pose 'CRCache_Input'())
    Alpha=1
    ConstraintWeights_0=0.4185
    ConstraintWeights_1:float=0.4001
    ConstraintWeights_2:float=0.1221
    ConstraintWeights_3:float=0.03
    ConstraintWeights_4:float=0.0293
   ) -> (Pose) {
    boneToModify {boneName=cape_01_L}
    constraintSetup [
     {targetBone {boneName=clavicle_l}}
     {targetBone {boneName=spine_05}}
     {targetBone {boneName=upperarm_l}}
     {targetBone {boneName=neck_01}}
     {targetBone {boneName=neck_02}}
    ]
    constraintWeights [1,1,1,1,1]
   }
   n11 = "Constraint - cape_01_C (PPP)"!(ComponentPose=@n10, Alpha=1, ConstraintWeights_0=0.8611, ConstraintWeights_1:float=0.1163, ConstraintWeights_2:float=0.0225) -> (Pose) {
    boneToModify {boneName=cape_01_C}
    constraintSetup [{targetBone {boneName=spine_05}}, {targetBone {boneName=neck_02}}, {targetBone {boneName=neck_01}}]
    constraintWeights [1,1,1]
   }
   n12 = "Constraint - cape_01_R (PPPPP)"!(
    ComponentPose=@n11
    Alpha=1
    ConstraintWeights_0=0.4185
    ConstraintWeights_1:float=0.4001
    ConstraintWeights_2:float=0.1221
    ConstraintWeights_3:float=0.03
    ConstraintWeights_4:float=0.0293
   ) -> (Pose) {
    boneToModify {boneName=cape_01_R}
    constraintSetup [
     {targetBone {boneName=clavicle_r}}
     {targetBone {boneName=spine_05}}
     {targetBone {boneName=upperarm_r}}
     {targetBone {boneName=neck_01}}
     {targetBone {boneName=neck_02}}
    ]
    constraintWeights [1,1,1,1,1]
   }
   n13 = "Constraint - rapier_attach (PPPP)"!(
    ComponentPose=@n12
    Alpha=1
    ConstraintWeights_0=0.6969
    ConstraintWeights_1:float=0.2014
    ConstraintWeights_2:float=0.084
    ConstraintWeights_3:float=0.0178
   ) -> (Pose) {
    boneToModify {boneName=rapier_attach}
    constraintSetup [{targetBone {boneName=pelvis}}, {targetBone {boneName=spine_01}}, {targetBone {boneName=thigh_l}}, {targetBone {boneName=spine_02}}]
    constraintWeights [1,1,1,1]
   }
   n14 = "Transform (Modify) Bone - Bone: rapier_attach"!(ComponentPose=@n13, Alpha=1, Translation="0.5,0,1", Rotation="0,5,0") -> (Pose) {rotationMode=BMM_Additive, translationMode=BMM_Additive, boneToModify {boneName=rapier_attach}, rotation {yaw=5}, translation {x=0.5, z=1}}
   n15 = "Transform (Modify) Bone - Bone: rapier_attach"(ComponentPose=@n14, Alpha=1, Translation="2.75,0,3.25", Rotation="16,-10,-5") {
    rotationMode=BMM_Additive
    translationMode=BMM_Additive
    boneToModify {boneName=rapier_attach}
    rotation=<rot 16 -10 -5>
    translation {x=2.75, z=3.25}
   }
   n16 = "Transform (Modify) Bone - Bone: cape_01_L"!(ComponentPose=@n15, Alpha=1, Rotation="-30,0,0") -> (Pose) {rotationMode=BMM_Additive, boneToModify {boneName=cape_01_L}, rotation {pitch=-30}}
   n17 = "Constraint - ring_01_l (RR)"!(
    ComponentPose="Transform (Modify) Bone - Bone: cape_01_R"!(
     ComponentPose="Transform (Modify) Bone - Bone: cape_01_C"!(ComponentPose=@n16, Alpha=1, Rotation="-30,0,0") -> (Pose) {rotationMode=BMM_Additive, boneToModify {boneName=cape_01_C}, rotation {pitch=-30}}
     Alpha=1
     Rotation="-30,0,0"
    ) -> (Pose) {rotationMode=BMM_Additive, boneToModify {boneName=cape_01_R}, rotation {pitch=-30}}
    Alpha=1
    ConstraintWeights_0=0.5
    ConstraintWeights_1:float=0.5
   ) -> (Pose) {
    boneToModify {boneName=ring_01_l}
    constraintSetup [{transformType=Rotation, targetBone {boneName=middle_01_l}}, {transformType=Rotation, targetBone {boneName=pinky_01_l}}]
    constraintWeights [1,1]
   }
   n18 = "Constraint - ring_02_l (RR)"!(ComponentPose=@n17, Alpha=1, ConstraintWeights_0=0.5, ConstraintWeights_1:float=0.5) -> (Pose) {
    boneToModify {boneName=ring_02_l}
    constraintSetup [{transformType=Rotation, targetBone {boneName=middle_02_l}}, {transformType=Rotation, targetBone {boneName=pinky_02_l}}]
    constraintWeights [1,1]
   }
   n19 = "Constraint - ring_03_l (RR)"!(ComponentPose=@n18, Alpha=1, ConstraintWeights_0=0.5, ConstraintWeights_1:float=0.5) -> (Pose) {
    boneToModify {boneName=ring_03_l}
    constraintSetup [{transformType=Rotation, targetBone {boneName=middle_03_l}}, {transformType=Rotation, targetBone {boneName=pinky_03_l}}]
    constraintWeights [1,1]
   }
   n20 = "Constraint - ring_01_r (RR)"!(ComponentPose=@n19, Alpha=1, ConstraintWeights_0=0.5, ConstraintWeights_1:float=0.5) -> (Pose) {
    boneToModify {boneName=ring_01_r}
    constraintSetup [{transformType=Rotation, targetBone {boneName=middle_01_r}}, {transformType=Rotation, targetBone {boneName=pinky_01_r}}]
    constraintWeights [1,1]
   }
   n21 = "Constraint - ring_02_r (RR)"!(ComponentPose=@n20, Alpha=1, ConstraintWeights_0=0.5, ConstraintWeights_1:float=0.5) -> (Pose) {
    boneToModify {boneName=ring_02_r}
    constraintSetup [{transformType=Rotation, targetBone {boneName=middle_02_r}}, {transformType=Rotation, targetBone {boneName=pinky_02_r}}]
    constraintWeights [1,1]
   }
   n22 = "Constraint - ring_03_r (RR)"!(ComponentPose=@n21, Alpha=1, ConstraintWeights_0=0.5, ConstraintWeights_1:float=0.5) -> (Pose) {
    boneToModify {boneName=ring_03_r}
    constraintSetup [{transformType=Rotation, targetBone {boneName=middle_03_r}}, {transformType=Rotation, targetBone {boneName=pinky_03_r}}]
    constraintWeights [1,1]
   }
   Save cached pose 'preDynamicsCache'(Pose=Component To Local(ComponentPose=@n22)) {cacheName=preDynamicsCache}
  ]}
 EventGraph {nodes [
   Event BlueprintUpdateAnimation()
   > Set GameTimeS(GameTimeS:double=GetGameTimeInSeconds()) -> (Output_Get:double)
   > n1 = Sequence()
   > Set L_Leg_Front(L_Leg_Front:float=GetAlphaFromReferenceTransformDelta(SkeletalMeshComponent=GetOwningComponent(), BoneName=thigh_l, BaseName=pelvis, InputRange=(X=0,Y=-100), TwistAxis="1,0,0")) -> (Output_Get:float)
   n2 = Get GameTimeS() -> (GameTimeS:double)
   n3 = RandomRotatorFunction!(
    execute=@n1.then_1
    GameTimeInSeconds:double=@n2
    PerlinScale:double=2
    Frequency:Rotator="2.5,3.5,3"
    Amplitude:Rotator="10,5,10"
    Offset:Rotator="0,0,25"
    TimeOffset:double=0
   ) -> (then, RandomRotator:Rotator)
   > Set L_CapeRot(L_CapeRot:Rotator=@n3) -> (Output_Get:Rotator)
   > n4 = RandomRotatorFunction(GameTimeInSeconds=@n2, PerlinScale=2, Frequency="3.5,2.5,3", Amplitude="-10,5,10", Offset="0,0,30", TimeOffset=1.25)
   > Set C_CapeRot(C_CapeRot:Rotator=@n4) -> (Output_Get:Rotator)
   > n5 = RandomRotatorFunction(GameTimeInSeconds=@n2, PerlinScale=2, Frequency="2.75,3.75,3", Amplitude="10,7.5,11", Offset="0,0,25", TimeOffset=0.75)
   > Set R_CapeRot(R_CapeRot:Rotator=@n5) -> (Output_Get:Rotator)
  ]}
]
variables [
 C_CapeRot {type=Rotator}
 GameTimeS {type=number}
 L_CapeRot {type=Rotator}
 L_Leg_Front {type=number}
 PoseAsset {type=PoseAsset, default=/CosmeticCompanions/Automation_Template/companion_pose_asset_template}
 R_CapeRot {type=Rotator}
]
